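# Sound cue for the demon's pre-emerge voice line in the GrillMoon event sequence.
# A single wave player feeds the GrillMoon_Seq_Demon_VO_Pre_Emerge wave straight to the output, with no randomization or mixing nodes.
# It plays under the Fort_Events sound class with boosted volume and the GrillMoon close point-of-interest attenuation for spatial falloff.
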SoundCue /GrillMoon/Assets/Audio/pjw/SC_GrillMoon_pjw_Demon_VO_Pre_Emerge
root=WavePlayer_1
nodes [{id=WavePlayer_1, properties {soundWaveAssetPtr=/GrillMoon/Assets/Audio/wav/GrillMoon_Seq_Demon_VO_Pre_Emerge}}]
properties {
 attenuationSettings=/GrillMoon_Audio/Assets/Settings/ATT_GrillMoon_POI_Close
 maxDistance=14000
 soundClassObject=&Fort_Events
 volumeMultiplier=2
}
