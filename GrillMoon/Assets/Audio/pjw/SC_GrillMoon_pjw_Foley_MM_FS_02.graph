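# Sound cue for a GrillMoon Foley footstep variant, MM_FS_02, in the project's audio set.
# A single wave player feeds the output directly from the GrillMoon_Foley_MM_FS_02 wave, with no randomization or modulation.
# It is spatialized through the close point-of-interest attenuation asset and routed to the Fort_Events sound class.
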SoundCue /GrillMoon/Assets/Audio/pjw/SC_GrillMoon_pjw_Foley_MM_FS_02
root=WavePlayer_0
nodes [{id=WavePlayer_0, properties {soundWaveAssetPtr=/GrillMoon/Assets/Audio/wav/GrillMoon_Foley_MM_FS_02}}]
properties {
 attenuationSettings=/GrillMoon_Audio/Assets/Settings/ATT_GrillMoon_POI_Close
 maxDistance=14000
 soundClassObject=&Fort_Events
 volumeMultiplier=1
}
