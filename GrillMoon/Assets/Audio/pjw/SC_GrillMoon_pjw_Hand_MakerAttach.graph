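# A Sound Cue in GrillMoon that plays a hand sound effect, named for a Maker attach moment in what appears to be a sequence.
# A single wave player feeds the GrillMoon_Seq_Maker_SFX_Hand wave straight to the output, with no randomization or mixing.
# It uses the medium point-of-interest attenuation and the Fort_Events sound class, with a slight volume boost.
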SoundCue /GrillMoon/Assets/Audio/pjw/SC_GrillMoon_pjw_Hand_MakerAttach
root=WavePlayer_0
nodes [{id=WavePlayer_0, properties {soundWaveAssetPtr=/GrillMoon/Assets/Audio/wav/GrillMoon_Seq_Maker_SFX_Hand}}]
properties {
 attenuationSettings=/GrillMoon_Audio/Assets/Settings/ATT_GrillMoon_POI_Med
 maxDistance=14000
 soundClassObject=&Fort_Events
 volumeMultiplier=1.2
}
